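# Translucent unlit Niagara material for the BiruFang character's spline-mesh arm flames.
# World position offset combines SplineThicken with a UV-weighted bend mixing sine wobble and clamped pull toward a dynamic-parameter position, which reads Position X for all three axes.
# A UV-shaped mask and panning wispy noise, varied by vertex red, drive particle-colored emissive and hard-edged opacity.
# Opacity fades along V, with depth fade only on default and deferred shading paths.
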Material /Game/Characters/Player/Male/Medium/Heads/M_MED_BiruFang_Head/FX/M_Character_BiruFang_SplineMeshArmFlames
graph {
 properties {
  bUsedWithNiagaraMeshParticles=true
  bUsedWithNiagaraRibbons=true
  blendMode=BLEND_Translucent
  pixelDepthOffsetMode=PDOM_Legacy
  shadingModel=MSM_Unlit
  usedShadingModels=Unlit
 }
 nodes [
  n1 = Time() {period=1}
  n2 = VertexColor()
  n3 = Add(
   A=ComponentMask(None=TextureCoordinate()) {g=true}
   B=Multiply(A=Multiply(A=@n1, B=@n2.R) {constB=0.5}, B=ScalarParameter() {defaultValue=1, parameterName=WPO Speed, sliderMax=1})
  )
  n4 = Add(
   A=Transform(None=Constant3Vector() {constant=<rgba 1 0 1 0>}) {transformSourceType=TRANSFORMSOURCE_Local}
   B=Constant3Vector() {constant=<rgba 0 0 1 0>}
  )
  n5 = ScalarParameter() {defaultValue=3, parameterName=WPO Upwards, sliderMax=50}
  n6 = DynamicParameter() {parameterIndex=1, paramNames [Position X, Position Y, Position Z, Param4]}
  n7 = Subtract(
   A=MaterialFunctionCall(X=@n6.Position X, Y=@n6.Position X, Z=@n6.Position X) {materialFunction=&MakeFloat3}.Result
   B=ParticlePositionWS()
  )
  n8 = ScalarParameter() {defaultValue=20, parameterName=WPO Max Offset, sliderMax=50}
  n9 = Multiply(A=@n8) {constB=-1}
  n10 = Add(
   A=Multiply(A=Sine(None=@n3), B=Multiply(A=@n4, B=@n5))
   B=Multiply(A=VectorParameter() {parameterName=WPO Axis Mask, defaultValue=<rgba 1 1 1 1>}.RGB, B=Clamp(None=@n7, Min=@n9, Max=@n8))
  )
  n11 = TextureCoordinate()
  n12 = Power(Base=OneMinus(None=ComponentMask(None=@n11) {g=true}), Exp=ScalarParameter() {defaultValue=1, parameterName=WPO Bend, sliderMax=10})
  Add(
   A=MaterialFunctionCall(
    WidthBase=ScalarParameter() {defaultValue=10, parameterName=Width Base}
    WidthTip=ScalarParameter() {defaultValue=10, parameterName=Width Tip}
    Expand U or V UV Channel=StaticBool() {value=true}
   ) {materialFunction=&SplineThicken}.WorldPositionOffset
   B=Multiply(A=ScalarParameter() {parameterName=WPO Intensity, sliderMax=100}, B=LinearInterpolate(B=@n10, Alpha=@n12))
  )
  n13 = ParticleColor()
  n14 = ComponentMask(None=TextureCoordinate()) {g=true}
  n15 = SphereMask(A=ComponentMask(None=TextureCoordinate()) {r=true}, B=Constant() {r=0.5}) {attenuationRadius=0.5, hardnessPercent=0}
  n16 = SmoothStep(Value=Power(Base=Multiply(A=@n14, B=@n15), Exp=ScalarParameter() {defaultValue=1, parameterName=Concave, sliderMax=10})) {constMax=0.75}
  n17 = ScalarParameter() {group=Wispy Flame, parameterName=Speed X, sliderMax=5, sliderMin=-5}
  n18 = ScalarParameter() {defaultValue=1, group=Wispy Flame, parameterName=Speed Y, sliderMax=5, sliderMin=-5}
  n19 = VertexColor()
  n20 = Panner(Coordinate=TextureCoordinate(), Speed=Multiply(A=AppendVector(A=@n17, B=@n18), B=@n19.R)) {speedX=1}
  n21 = AppendVector(
   A=ScalarParameter() {defaultValue=1, group=Wispy Flame, parameterName=Scale X, sliderMax=5}
   B=ScalarParameter() {defaultValue=1, group=Wispy Flame, parameterName=Scale Y, sliderMax=5}
  )
  n22 = Constant()
  n23 = MaterialFunctionCall(XY=Add(A=Multiply(A=@n20, B=@n21), B=AppendVector(A=@n19.R, B=@n22))) {materialFunction=&Swizzle}
  n24 = LinearInterpolate(
   A=Multiply(
    A=Multiply(A=@n16) {constB=2}
    B=TextureSampleParameter2D(UVs=@n23.YX) {group=Wispy Flame, parameterName=Flame Texture, texture=/Game/Effects/Fort_Effects/Textures/T_WispyFireNoise03}.R
   )
   B=@n16
   Alpha=@n16
  )
  MaterialFunctionCall(Emissive=LinearInterpolate(A=Multiply(A=@n13.RGB) {constB=0.5}, B=@n13.RGB, Alpha=@n24) {constB=0}) {materialFunction=&MF_FX_EmissiveAdjustments}
  n25 = SmoothStep(Max=OneMinus(None=ScalarParameter() {defaultValue=0.9, parameterName=Edge Hardness, sliderMax=1}), Value=Multiply(A=@n13.A, B=@n24)) {constMax=0.1}
  n26 = DepthFade(Opacity=@n25, FadeDistance=ScalarParameter() {defaultValue=35, parameterName=Depth Fade Distance, sliderMax=100}) {fadeDistanceDefault=25}
  n27 = ComponentMask(None=TextureCoordinate()) {g=true}
  Multiply(
   A=Multiply(
    A=ShadingPathSwitch(Default=@n26, Deferred=@n26, Forward=@n25, Mobile=@n25)
    B=ScalarParameter() {defaultValue=1, parameterName=Opacity, sliderMax=1}
   )
   B=SmoothStep(Max=ScalarParameter() {parameterName=Edge UV Fade, sliderMax=10}, Value=OneMinus(None=@n27))
  )
  VectorParameter() {defaultValue=<rgba 0 0 0 1>}
 ]
}
